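; Looping gameplay cue for the active state of the Plum item's KineticAbs effect, apparently kinetic absorption, driving its Niagara visual intensity.
; When looping starts, it plays a timeline whose float track is written each update to the "EffectIntensity" variable on the first spawned particle component.
; BeginPlay, Tick, overlap and construction script are empty.
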(event EventBeginPlay)

(event EventOnLoopingStart(Niagara)
  (params (MyTarget :type Actor%20Object%20Reference) (Parameters :type Gameplay%20Cue%20Parameters%20Structure%20%28by%20ref%29) (ParticleComponents :type Array%20of%20Niagara%20Particle%20System%20Component%20Object%20References) (AudioComponents :type Array%20of%20Audio%20Component%20Object%20References))
  (bind (_direction _intensity) (Timeline|Timeline_0 Play)
    (:Update
      (Niagara|SetNiagaraVariableByString(Float) (Utilities|Array|Get(acopy) ParticleComponents) "EffectIntensity" _intensity))))

(event Collision|EventActorBeginOverlap
  (params (OtherActor :type Actor%20Object%20Reference)))

(event EventTick
  (params (DeltaSeconds :type Float%20%28single-precision%29)))

(fn ConstructionScript)
