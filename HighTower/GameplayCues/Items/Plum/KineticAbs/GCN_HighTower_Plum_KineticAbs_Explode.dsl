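; Burst gameplay cue for the HighTower Plum kinetic absorption explosion that adapts its Niagara effect to the surface under the target Fort player pawn.
; If the pawn is surface swimming, it moves the effect to the water surface, enables water spawning, and adds a lens splash and water sound.
; Otherwise a vertical line trace enables debris only when the hit physical material is in the allowed list, with water off.
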(fn OnBurst(Niagara)
  (params (MyTarget :type Actor%20Object%20Reference) (Parameters :type Gameplay%20Cue%20Parameters%20Structure%20%28by%20ref%29) (ParticleComponents :type Array%20of%20Niagara%20Particle%20System%20Component%20Object%20References) (AudioComponents :type Array%20of%20Audio%20Component%20Object%20References) (BurstCameraShakeInstance :type Camera%20Shake%20Base%20Object%20Reference) (BurstDecalInstance :type Decal%20Actor%20Object%20Reference))
  (bind _output (Utilities|Array|Get(acopy) ParticleComponents))
  (bind _asfort_player_pawn (Utilities|Casting|CastToFortPlayerPawn MyTarget)
    (:then
      (if (WaterBody|IsSurfaceSwimming _asfort_player_pawn)
        ;[1 1 1] In Water
        (bind _outwatersurfaceinfo (WaterBody|GetWaterSurfaceInfoStruct (WaterBody|GetCurrentWaterBody _asfort_player_pawn) (Transformation|GetWorldLocation (Variables|/Script/Engine.Character|GetMesh _asfort_player_pawn)) false false false))
        (bind _outwatersurfaceinfo_waterplanelocation (Utilities|Struct|BreakWaterSurfaceInfo _outwatersurfaceinfo))
        (bind (_returnvalue_x _returnvalue_y) (Math|Vector|BreakVector (Transformation|GetSocketLocation (Variables|/Script/Engine.Character|GetMesh _asfort_player_pawn) "root")))
        (Transformation|SetWorldLocation _output (Math|Vector|MakeVector _returnvalue_x _returnvalue_y (.z _outwatersurfaceinfo_waterplanelocation)) false false)
        (Niagara|SetNiagaraVariable(Bool) _output "SpawnDebris" false)
        (Niagara|SetNiagaraVariable(Bool) _output "SpawnWater" true)
        (Feedback|SpawnGenericCameraLensEffect _asfort_player_pawn "/Game/Effects/Fort_Effects/Camera_Effects/B_CameraLens_Splash_Water_Droplets.B_CameraLens_Splash_Water_Droplets_C")
        (Audio|PlaySoundatLocation _asfort_player_pawn "/HighTower/Audio/Plum/KineticAbsorption/Plum_Expl_Debris_Water_Cue.Plum_Expl_Debris_Water_Cue" (Transformation|GetActorLocation _asfort_player_pawn))
        (else
          (Niagara|SetNiagaraVariable(Bool) _output "SpawnWater" false)
          (bind (_outhit _returnvalue) (Collision|LineTraceByChannel :WorldContextObject MyTarget :Start (+ (Transformation|GetWorldLocation _output) "0, 0,200.000000") :End (+ (Transformation|GetWorldLocation _output) "0, 0,-500.000000") :TraceChannel "TraceTypeQuery10" :bTraceComplex false :DrawDebugType "None" :bIgnoreSelf true :TraceColor "(R=1.000000,G=0.000000,B=0.000000,A=1.000000)" :TraceHitColor "(R=0.000000,G=1.000000,B=0.000000,A=1.000000)" :DrawTime 3e+01))
          (if _returnvalue
            (Development|PrintString 0 (Utilities|GetDisplayName (.PhysMat (Collision|BreakHitResult _outhit))) false false "(R=0.552083,G=0.387610,B=0.159631,A=1.000000)" 2e+01)
            (if (Utilities|Array|ContainsItem (Variables|Default|GetPhysMat) (.PhysMat (Collision|BreakHitResult _outhit)))
              (Niagara|SetNiagaraVariable(Bool) _output "SpawnDebris" true)
              (else
                (Niagara|SetNiagaraVariable(Bool) _output "SpawnDebris" false)))
            (else
              (Niagara|SetNiagaraVariable(Bool) _output "SpawnDebris" false)
              (Niagara|SetNiagaraVariable(Bool) _output "SpawnWater" false)))))
      (return))
    (:CastFailed
      (Development|PrintString 0 "CastFailed" true true "(R=0.000000,G=0.143090,B=1.000000,A=1.000000)" 6.0))))
